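# Physics asset for the SleepyUni male medium player body cosmetic, built around two dangling pants drawstring cords.
# Each cord is a two-link capsule chain hanging from a cord base body on the pelvis, with tight angular limits and a wider second swing on the tip links.
# Pelvis and thigh capsules give the lower-body shape, and their joints and the cord base joint leave swing and twist free.
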PhysicsAsset /BRCosmetics/Characters/Player/Male/Medium/Bodies/M_MED_SleepyUni/Meshes/M_MED_SleepyUni_Physics
bodies {
 base_pants_Cord [{shapeType=Capsule, length=0.505, radius=0.505, shapeName=base_pants_Cord_capsule, rotation=<rot 0 -90 0>}]
 dyn_pantsCord_l_01 [{shapeType=Capsule, length=3, radius=1, center=<2 0 0>, rotation=<rot 90 0 0>}]
 dyn_pantsCord_l_02 [{shapeType=Capsule, length=3, radius=1, center=<3 0 0>, rotation=<rot 90 0 0>}]
 dyn_pantsCord_r_01 [{shapeType=Capsule, length=3, radius=1, center=<2 0 0>, rotation=<rot 90 0 0>}]
 dyn_pantsCord_r_02 [{shapeType=Capsule, length=3, radius=1, center=<3 0 0>, rotation=<rot 90 0 0>}]
 pelvis [{shapeType=Capsule, length=7.8158, radius=9.7677, shapeName=pelvis_capsule, center=<-1.7087 -0.7696 0>}]
 thigh_l [{
   shapeType=Capsule
   length=18.2256
   radius=8.698
   shapeName=thigh_l_capsule
   center=<-11.2671 0.4516 1.4105>
   rotation=<rot 86.9791 -172.0403 11.0514>
  }]
 thigh_r [{
   shapeType=Capsule
   length=18.2145
   radius=8.3124
   shapeName=thigh_r_capsule
   center=<9.9246 0.2943 -0.6736>
   rotation=<rot 86.7498 -172.8292 -169.7364>
  }]
}
constraints [
 {
  bone1Name=thigh_r
  bone2Name=pelvis
  swing1LimitDegrees=45
  swing1Motion=Free
  swing2LimitDegrees=45
  swing2Motion=Free
  twistLimitDegrees=45
  twistMotion=Free
 }
 {
  bone1Name=base_pants_Cord
  bone2Name=pelvis
  swing1LimitDegrees=45
  swing1Motion=Free
  swing2LimitDegrees=45
  swing2Motion=Free
  twistLimitDegrees=45
  twistMotion=Free
 }
 {bone1Name=dyn_pantsCord_l_01, bone2Name=base_pants_Cord, swing1LimitDegrees=5, swing2LimitDegrees=10, twistLimitDegrees=3}
 {bone1Name=dyn_pantsCord_l_02, bone2Name=dyn_pantsCord_l_01, swing1LimitDegrees=5, swing2LimitDegrees=30, twistLimitDegrees=3}
 {bone1Name=dyn_pantsCord_r_01, bone2Name=base_pants_Cord, swing1LimitDegrees=5, swing2LimitDegrees=10, twistLimitDegrees=3}
 {bone1Name=dyn_pantsCord_r_02, bone2Name=dyn_pantsCord_r_01, swing1LimitDegrees=5, swing2LimitDegrees=30, twistLimitDegrees=3}
 {
  bone1Name=thigh_l
  bone2Name=pelvis
  swing1LimitDegrees=45
  swing1Motion=Free
  swing2LimitDegrees=45
  swing2Motion=Free
  twistLimitDegrees=45
  twistMotion=Free
 }
]
